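# Editor utility widget that shows MetaSound inputs in a collapsible "Inputs" list with a search toggle, built on TechAudioToolsEditorListView.
# Toggling search on expands the area and focuses the search box; toggling off clears it, and collapsing resets both.
# Expansion changes are re-broadcast through an On Expansion Changed dispatcher.
# Item clicks broadcast a field notification for Get Selected Input Editor Viewmodel, which casts the list viewmodel's selected entry data.
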EditorUtilityWidgetBlueprint /TechAudioToolsContent/Widgets/Editor/Lists/MetaSoundMemberLists/EUW_MetaSoundInputList_Expandable
graphs [
 EventGraph {nodes [
   n1 = "On Expansion Changed (Inputs_ExpandableArea)"!() -> (OutputDelegate, then, Area:Expandable Area Object Reference, bIsExpanded:bool) {is=title K2Node_ComponentBoundEvent}
   > Call On Expansion Changed!(execute:Exec, self:Meta Sound Input List Object Reference, Is Expanded:bool=@n1.bIsExpanded) -> (then:Exec) {is=title K2Node_CallDelegate}
   > n2 = Branch(Condition=@n1.bIsExpanded)
   Set Checked State!(
    execute=@n2.else
    self:EUW Meta Sound Member List Search Toggle Button Object Reference=Get Search_ToggleButton() -> (Search_ToggleButton:EUW Meta Sound Member List Search Toggle Button Object Reference)
    Is Checked:bool=false
   ) -> (then) {is=title K2Node_CallFunction}
   > Clear Text(self:Tech Audio Tools Search Box Object Reference=Get EUW_SearchBox() -> (EUW_SearchBox:Tech Audio Tools Search Box Object Reference))
   n3 = "On Checked State Changed (Search_ToggleButton)"!() -> (OutputDelegate, then, Is Checked:bool) {is=title K2Node_ComponentBoundEvent}
   > n4 = Branch(Condition=@n3.Is Checked)
   > SetIsExpanded!(
    execute
    self:Expandable Area Object Reference=Get Inputs_ExpandableArea() -> (Inputs_ExpandableArea:Editor Utility Expandable Area Object Reference)
    IsExpanded:bool=true
   ) -> (then)
   > SetKeyboardFocus(self:Widget Object Reference=Get EditableTextBox!(self:Tech Audio Tools Search Box Object Reference=Get EUW_SearchBox()) -> (EditableTextBox:Editor Utility Editable Text Box Object Reference))
   Clear Text(execute=@n4.else, self=Get EUW_SearchBox())
   "On Item Clicked (ListView)"!() -> (OutputDelegate, then, Item:Object Reference) {is=title K2Node_ComponentBoundEvent}
   > Broadcast Field Value Changed!(execute, self:Widget Object Reference, FieldId:Field Notification Id Structure=(FieldName="Get Selected Input Editor Viewmodel")) -> (then)
   Event PreConstruct()
   n5 = Get Search_ToggleButton()
   > SetVisibility(
    self=Get Search_Border() -> (Search_Border:Border Object Reference)
    InVisibility=Select(
     Option 0=Collapsed
     Option 1=Visible
     Index=Is Checked!(self:EUW Meta Sound Member List Search Toggle Button Object Reference=@n5) -> (Return Value:bool)
    )
   )
  ]}
 Get Selected Input Editor Viewmodel {nodes [
   Get Selected Input Editor Viewmodel() {is=title K2Node_FunctionEntry}
   n1 = Get TechAudioToolsListViewModel() -> (TechAudioToolsListViewModel:Tech Audio Tools List Viewmodel Object Reference)
   > n2 = Cast To MetaSoundInputEditorViewModel!(
    execute
    Object=Get EntryData!(self:Tech Audio Tools Entry Viewmodel Object Reference=Get SelectedEntryViewModel!(self:Tech Audio Tools List Viewmodel Object Reference=@n1) -> (SelectedEntryViewModel:Tech Audio Tools Entry Viewmodel Object Reference)) -> (EntryData:Object Reference)
   ) -> (then, CastFailed, AsMetaSound Input Editor Viewmodel:Meta Sound Input Editor Viewmodel Object Reference) {is=title K2Node_DynamicCast}
   > Return Node!(execute, Input Editor Viewmodel:Meta Sound Input Editor Viewmodel Object Reference=@n2)
  ]}
 On Expansion Changed {nodes [On Expansion Changed!() -> (then, Is Expanded:bool=~) {is=title K2Node_FunctionEntry}]}
]
variables [On Expansion Changed {default=()}]
widgets {
 info {
  inheritedWidgetCount=1
  namedSlotCount=2
  parentClass=/Script/TechAudioToolsContentEditor.TechAudioToolsEditorListView
  rootWidgetClass=/Script/Blutility.EditorUtilityExpandableArea
  widgetCount=8
 }
 widgets [
  Inputs_ExpandableArea {
   widgetClassPath=&EditorUtilityExpandableArea
   bIsVariable=true
   widget=Inputs_ExpandableArea
   properties {
    bIsExpanded=true
    areaPadding=<ltrb 0 0 0 0>
    borderBrush {
     bIsDynamicallyLoaded=false
     drawAs=Box
     imageType=NoImage
     mirroring=NoMirror
     resourceName=None
     resourceObject=null
     tiling=NoTile
     imageSize=<0 0>
     margin=<ltrb 0 0 0 0>
     outlineSettings {
      bUseBrushTransparency=false
      roundingType=FixedRadius
      width=0
      color {colorUseRule=UseColor_Specified, specifiedColor=<rgba 0 0 0 0>}
      cornerRadii=<4 4 4 4>
     }
     tintColor {colorUseRule=UseColor_Specified, specifiedColor=<rgba 0.0284 0.0284 0.0284 1>}
     uVRegion {bIsValid=false, max=<0 0>, min=<0 0>}
    }
    headerPadding=<ltrb 8 4 8 4>
    style {
     rolloutAnimationSeconds=0
     collapsedImage {
      bIsDynamicallyLoaded=false
      drawAs=Image
      imageType=Vector
      mirroring=NoMirror
      resourceName=../../../Engine/Content/Slate/Starship/Common/caret-right.svg
      resourceObject=null
      tiling=NoTile
      imageSize=<16 16>
      margin=<ltrb 0 0 0 0>
      outlineSettings {
       bUseBrushTransparency=false
       roundingType=HalfHeightRadius
       width=0
       color {colorUseRule=UseColor_Specified, specifiedColor=<rgba 0 0 0 0>}
       cornerRadii=<0 0 0 0>
      }
      tintColor {colorUseRule=UseColor_Specified, specifiedColor=<rgba 0.5271 0.5271 0.5271 1>}
      uVRegion {bIsValid=false, max=<0 0>, min=<0 0>}
     }
     expandedImage {
      bIsDynamicallyLoaded=false
      drawAs=Image
      imageType=Vector
      mirroring=NoMirror
      resourceName=../../../Engine/Content/Slate/Starship/Common/caret-down.svg
      resourceObject=null
      tiling=NoTile
      imageSize=<16 16>
      margin=<ltrb 0 0 0 0>
      outlineSettings {
       bUseBrushTransparency=false
       roundingType=HalfHeightRadius
       width=0
       color {colorUseRule=UseColor_Specified, specifiedColor=<rgba 0 0 0 0>}
       cornerRadii=<0 0 0 0>
      }
      tintColor {colorUseRule=UseColor_Specified, specifiedColor=<rgba 0.5271 0.5271 0.5271 1>}
      uVRegion {bIsValid=false, max=<0 0>, min=<0 0>}
     }
    }
   }
  }
  Header_Box {widgetClassPath=&HorizontalBox, bIsVariable=true, namedSlotHost=Inputs_ExpandableArea, widget=Header_Box}
  InputsHeader {
   widgetClassPath=&TextBlock
   parent=Header_Box
   widget=InputsHeader
   properties {
    text=Inputs
    textTransformPolicy=ToUpper
    colorAndOpacity {colorUseRule=UseColor_Specified, specifiedColor=<rgba 0.5776 0.5776 0.5776 1>}
    font {
     bForceMonospaced=false
     bMaterialIsStencil=false
     fontMaterial=null
     fontObject=&Roboto
     letterSpacing=0
     monospacedWidth=1
     size=8.25
     skewAmount=0
     typefaceFontName=Bold
     outlineSettings {
      bApplyOutlineToDropShadows=false
      bMiteredCorners=false
      bSeparateFillAlpha=false
      outlineMaterial=null
      outlineSize=0
      outlineColor=<rgba 0 0 0 1>
     }
    }
   }
   slot {class=HorizontalBoxSlot, properties {verticalAlignment=VAlign_Center, size {sizeRule=Fill, value=1}}}
  }
  Search_ToggleButton {
   widgetClassPath=&/EUW_MetaSoundMemberList_SearchToggleButton
   bIsVariable=true
   parent=Header_Box
   widget=Search_ToggleButton
   slot {class=HorizontalBoxSlot}
  }
  VerticalBox_0 {widgetClassPath=&VerticalBox, namedSlotHost=Inputs_ExpandableArea, widget=VerticalBox_0}
  Search_Border {
   widgetClassPath=&Border
   bIsVariable=true
   parent=VerticalBox_0
   widget=Search_Border
   properties {brushColor=<rgba 0.0284 0.0284 0.0284 1>, padding=<ltrb 2 2 2 2>}
   slot {class=VerticalBoxSlot}
  }
  EUW_SearchBox {
   widgetClassPath=/TechAudioToolsContent/Widgets/Editor/EUW_SearchBox
   bIsVariable=true
   parent=Search_Border
   widget=EUW_SearchBox
   slot {class=BorderSlot, properties {padding=<ltrb 2 2 2 2>}}
  }
  ListView {
   widgetClassPath=/Script/Blutility.EditorUtilityListView
   bInherited=true
   bIsVariable=true
   parent=VerticalBox_0
   widget=ListView
   properties {
    bEnableRightClickScrolling=false
    bEnableScrollAnimation=true
    entryWidgetClass=&/EUW_MetaSoundInputEntry.EUW_MetaSoundInputEntry_C
    numDesignerPreviewEntries=15
   }
   slot {class=VerticalBoxSlot, properties {size {sizeRule=Fill, value=1}}}
  }
 ]
}
