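# Editor utility widget that serves as a single text entry, apparently for the MetaSound member list sort menu.
# A horizontal box holds one small text block as the option label.
# On Construct, it copies its Option Text string variable into that label, replacing the placeholder text.
# It contains no selection or sorting logic of its own.
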EditorUtilityWidgetBlueprint /TechAudioToolsContent/Widgets/Editor/Lists/MetaSoundMemberLists/EUW_MetaSoundSortMenu_Entry
graphs [EventGraph {nodes [
   Event Construct()
   > "SetText (Text)"(self=Get Option_Text() -> (Option_Text:Text Object Reference), InText=Get Option Text() -> (Option Text:Text))
  ]}]
variables [Option Text {type=string, default=""}]
widgets {
 info {rootWidgetClass=&HorizontalBox, widgetCount=2}
 widgets [
  HorizontalBox_0 {widgetClassPath=&HorizontalBox, widget=HorizontalBox_0}
  Option_Text {
   widgetClassPath=&TextBlock
   bIsVariable=true
   parent=HorizontalBox_0
   widget=Option_Text
   properties {
    text=Option
    font {
     bForceMonospaced=false
     bMaterialIsStencil=false
     fontMaterial=null
     fontObject=&Roboto
     letterSpacing=0
     monospacedWidth=1
     size=9
     skewAmount=0
     typefaceFontName=Regular
     outlineSettings {
      bApplyOutlineToDropShadows=false
      bMiteredCorners=false
      bSeparateFillAlpha=false
      outlineMaterial=null
      outlineSize=0
      outlineColor=<rgba 0 0 0 1>
     }
    }
   }
   slot {class=HorizontalBoxSlot}
  }
 ]
}
