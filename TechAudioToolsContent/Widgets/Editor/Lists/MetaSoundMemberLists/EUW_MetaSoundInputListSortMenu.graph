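# Editor utility widget that provides the sort menu for a MetaSound input list: a sort-method combo box and an ascending/descending toggle with icon.
# Construct events fill the combo from the method enum, map names back to enum values, and sync the toggle from the list viewmodel's sort order.
# Toggling writes the viewmodel's SortOrder; method changes broadcast a field notification for Get Selected Sort Method.
# Both changes fire the On Sort Changed dispatcher.
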EditorUtilityWidgetBlueprint /TechAudioToolsContent/Widgets/Editor/Lists/MetaSoundMemberLists/EUW_MetaSoundInputListSortMenu
graphs [
 EventGraph {nodes [
   "On Selection Changed (SortMethod_ComboBox)"!() -> (OutputDelegate, then, SelectedItem:Name=None, SelectionType:ESelectInfo Enum) {is=title K2Node_ComponentBoundEvent}
   > Broadcast Field Value Changed!(execute, self:Widget Object Reference, FieldId:Field Notification Id Structure=(FieldName="Get Selected Sort Method")) -> (then)
   > Call On Sort Changed!(execute:Exec, self:EUW Meta Sound Input List Sort Menu Object Reference) -> (then:Exec) {is=title K2Node_CallDelegate}
   n1 = "On Check State Changed (SortOrder_ToggleButton)"!() -> (OutputDelegate, then, bIsChecked:bool) {is=title K2Node_ComponentBoundEvent}
   > n2 = Get!(execute:Exec) -> (then:Exec, else:Exec, MetaSoundInputListEditorViewModel:Meta Sound Input List Editor Viewmodel Object Reference)
   > Set SortOrder!(
    execute
    SortOrder:EArraySortOrder Enum=Select(Option 0:EArraySortOrder Enum=Ascending, Option 1:EArraySortOrder Enum=Descending, Index=@n1.bIsChecked) -> (ReturnValue:EArraySortOrder Enum)
    self:Meta Sound Member List Viewmodel Object Reference=@n2
   ) -> (then, Output_Get:EArraySortOrder Enum=Ascending)
   > Call On Sort Changed()
   n3 = Event PreConstruct()
   n4 = For Each Loop(
    Exec=[@n3.then, @n7.then]
    Array:Array of EMetaSoundMemberEditorSortMethod Enums=Get Sort Method Options() -> (Sort Method Options:Array of EMetaSoundMemberEditorSortMethod Enums=Alphabetical)
   ) -> (Array Element:EMetaSoundMemberEditorSortMethod Enum)
   SetSelectedOption!(
    execute=@n4.Completed
    self:"Combo Box (Key) Object Reference"=Get SortMethod_ComboBox() -> (SortMethod_ComboBox:Editor Utility Combo Box Key Object Reference)
    Option:Name=Sort Order
   ) -> (then)
   > n5 = Get()
   > SetIsChecked!(
    execute
    self:Check Box Object Reference=Get SortOrder_ToggleButton() -> (SortOrder_ToggleButton:Editor Utility Check Box Object Reference)
    InIsChecked:bool=Select!(
     Ascending:bool=false
     Descending:bool=true
     Index:EArraySortOrder Enum=Get SortOrder!(self:Meta Sound Member List Viewmodel Object Reference=@n5) -> (SortOrder:EArraySortOrder Enum=Ascending)
    ) -> (ReturnValue:bool)
   ) -> (then)
   n6 = String To Name(InString=Enum to String!(Enumerator:EMetaSoundMemberEditorSortMethod Enum=@n4) -> (ReturnValue:String))
   AddOption!(execute=@n4.LoopBody, self:"Combo Box (Key) Object Reference"=Get SortMethod_ComboBox(), Option:Name=@n6) -> (then)
   > Add!(
    execute
    TargetMap:Map of Names to EMetaSoundMemberEditorSortMethod Enums=Get Sort Method Options Map() -> (Sort Method Options Map:Map of Names to EMetaSoundMemberEditorSortMethod Enums)
    Key:"Name (by ref)"=@n6
    Value:"EMetaSoundMemberEditorSortMethod Enum (by ref)"=@n4
   ) -> (then) {is=title K2Node_CallFunction}
   n7 = Event Construct()
  ]}
 Get Selected Sort Method {nodes [
   Get Selected Sort Method() {is=title K2Node_FunctionEntry}
   n1 = Find(
    TargetMap:Map of Names to EMetaSoundMemberEditorSortMethod Enums=Get Sort Method Options Map()
    Key:"Name (by ref)"=GetSelectedOption!(self:"Combo Box (Key) Object Reference"=Get SortMethod_ComboBox()) -> (ReturnValue:Name=None)
   ) -> (Value:EMetaSoundMemberEditorSortMethod Enum=Alphabetical)
   > Return Node!(
    execute
    SelectedSortMethod:EMetaSoundMemberEditorSortMethod Enum=Select!(Option 0:EMetaSoundMemberEditorSortMethod Enum=SortOrder, Option 1:EMetaSoundMemberEditorSortMethod Enum=@n1, Index:bool=@n1.ReturnValue) -> (ReturnValue:EMetaSoundMemberEditorSortMethod Enum)
   )
  ]}
 Get Selected Sort Order {nodes [
   Get Selected Sort Order() {is=title K2Node_FunctionEntry}
   n1 = Get SortOrder_ToggleButton()
   > Return Node!(
    execute
    ReturnValue:EArraySortOrder Enum=Select(Option 0:EArraySortOrder Enum=Ascending, Option 1:EArraySortOrder Enum=Descending, Index=IsChecked(self=@n1)) -> (ReturnValue:EArraySortOrder Enum)
   )
  ]}
 Get Sort Method Icon {nodes [
   n1 = Get Sort Method Icon!() -> (then, Sort Method:EMetaSoundMemberEditorSortMethod Enum=Alphabetical, Sort Order:EArraySortOrder Enum=Ascending) {is=title K2Node_FunctionEntry}
   > Return Node!(
    execute
    Icon Path:Name=Select!(
     Ascending:Name=Select!(
      Alphabetical:Name=EditorStyle.LandscapeEditor.Target_DisplayOrder.Alphabetical.Ascending
      DataType:Name=EditorStyle.LandscapeEditor.Target_DisplayOrder.Default.Descending
      SortOrder:Name=EditorStyle.LandscapeEditor.Target_DisplayOrder.Default.Descending
      AddedOrder:Name=EditorStyle.LandscapeEditor.Target_DisplayOrder.Default.Descending
      Index:EMetaSoundMemberEditorSortMethod Enum=@n1
     ) -> (ReturnValue)
     Descending:Name=Select(
      Alphabetical=EditorStyle.LandscapeEditor.Target_DisplayOrder.Alphabetical.Descending
      DataType=EditorStyle.LandscapeEditor.Target_DisplayOrder.Default.Ascending
      SortOrder=EditorStyle.LandscapeEditor.Target_DisplayOrder.Default.Ascending
      AddedOrder=EditorStyle.LandscapeEditor.Target_DisplayOrder.Default.Ascending
      Index=@n1
     )
     Index:EArraySortOrder Enum=@n1.Sort Order
    ) -> (ReturnValue:Name)
   )
  ]}
 Get Sort Order Tool Tip Text {nodes [
   n1 = Get Sort Order Tool Tip Text!() -> (then, Sort Order:EArraySortOrder Enum=Ascending) {is=title K2Node_FunctionEntry}
   > Return Node!(
    execute
    Tool Tip:Text=Select!(Ascending:Text="LOCTEXT(\"Ascending\")", Descending:Text="LOCTEXT(\"Descending\")", Index:EArraySortOrder Enum=@n1) -> (ReturnValue:Text)
   )
  ]}
 On Sort Changed {nodes [On Sort Changed() {is=title K2Node_FunctionEntry}]}
 On_SortMethod_ComboBox_GenerateContentWidget {nodes [
   n1 = On_SortMethod_ComboBox_GenerateContentWidget!() -> (then, Item:Name) {is=title K2Node_FunctionEntry}
   > n2 = Create Widget!(
    execute:Exec
    WorldContextObject:Object Reference
    Class:User Widget Class Reference=&/EUW_MetaSoundSortMenu_Entry.EUW_MetaSoundSortMenu_Entry_C
    OwningPlayer:Player Controller Object Reference
    Option Text:Text="To Text (Name)"!(InName:Name=@n1) -> (ReturnValue:Text)
   ) -> (then:Exec, ReturnValue:EUW Meta Sound Sort Menu Entry Object Reference)
   > Return Node!(execute, ReturnValue:Widget Object Reference=@n2)
  ]}
 On_SortMethod_ComboBox_GenerateItemWidget {nodes [
   n1 = On_SortMethod_ComboBox_GenerateItemWidget!() -> (then, Item:Name) {is=title K2Node_FunctionEntry}
   > n2 = Create Widget(
    Class=&EUW_MetaSoundSortMenu_Entry_C
    Option Text="To Text (Name)"(InName=@n1)
   )
   > Return Node(ReturnValue=@n2)
  ]}
 Set Sort Method {nodes [
   n1 = Set Sort Method!() -> (then, Sort Method:EMetaSoundMemberEditorSortMethod Enum=Alphabetical) {is=title K2Node_FunctionEntry}
   > SetSelectedOption(self=Get SortMethod_ComboBox(), Option=String To Name(InString=Enum to String(Enumerator=@n1)))
  ]}
]
variables [
 On Sort Changed {default=()}
 Sort Method Options {type="array<EMetaSoundMemberEditorSortMethod>", default [SortOrder,Alphabetical,DataType,AddedOrder]}
 Sort Method Options Map {type="map<EMetaSoundMemberEditorSortMethod>", default {}}
]
widgets {
 info {rootWidgetClass=&VerticalBox, widgetCount=6}
 widgets [
  VerticalBox_1 {widgetClassPath=&VerticalBox, widget=VerticalBox_1}
  HorizontalBox_3 {
   widgetClassPath=&HorizontalBox
   parent=VerticalBox_1
   widget=HorizontalBox_3
   properties {bOverride_Cursor=true}
   slot {class=VerticalBoxSlot}
  }
  SizeBox_0 {
   widgetClassPath=&SizeBox
   parent=HorizontalBox_3
   widget=SizeBox_0
   properties {widthOverride=108}
   slot {class=HorizontalBoxSlot}
  }
  SortMethod_ComboBox {
   widgetClassPath=/Script/Blutility.EditorUtilityComboBoxKey
   bIsVariable=true
   parent=SizeBox_0
   widget=SortMethod_ComboBox
   properties {
    selectedOption=Sort Order
    contentPadding=<ltrb 8 2 4 2>
    foregroundColor {colorUseRule=UseColor_Specified, specifiedColor=<rgba 0.5271 0.5271 0.5271 1>}
    widgetStyle {
     comboButtonStyle {
      downArrowAlign=VAlign_Center
      buttonStyle {
       clickedSlateSound {resourceObject=null}
       disabled {
        bIsDynamicallyLoaded=false
        drawAs=RoundedBox
        imageType=NoImage
        mirroring=NoMirror
        resourceName=None
        resourceObject=null
        tiling=NoTile
        imageSize=<0 0>
        margin=<ltrb 0 0 0 0>
        outlineSettings {
         bUseBrushTransparency=false
         roundingType=FixedRadius
         width=1
         color {colorUseRule=UseColor_ColorTable, specifiedColor=<rgba 0 0 0 0>}
         cornerRadii=<4 4 4 4>
        }
        tintColor {colorUseRule=UseColor_Specified, specifiedColor=<rgba 0.0395 0.0395 0.0395 1>}
        uVRegion {bIsValid=false, max=<0 0>, min=<0 0>}
       }
       disabledForeground {colorUseRule=UseColor_ColorTable, specifiedColor=<rgba 0 0 0 0>}
       hovered {
        bIsDynamicallyLoaded=false
        drawAs=RoundedBox
        imageType=NoImage
        mirroring=NoMirror
        resourceName=None
        resourceObject=null
        tiling=NoTile
        imageSize=<0 0>
        margin=<ltrb 0 0 0 0>
        outlineSettings {
         bUseBrushTransparency=false
         roundingType=FixedRadius
         width=1
         color {colorUseRule=UseColor_ColorTable, specifiedColor=<rgba 0 0 0 0>}
         cornerRadii=<4 4 4 4>
        }
        tintColor {colorUseRule=UseColor_Specified, specifiedColor=<rgba 0.0048 0.0048 0.0048 1>}
        uVRegion {bIsValid=false, max=<0 0>, min=<0 0>}
       }
       hoveredForeground {colorUseRule=UseColor_ColorTable, specifiedColor=<rgba 0 0 0 0>}
       hoveredSlateSound {resourceObject=null}
       normal {
        bIsDynamicallyLoaded=false
        drawAs=RoundedBox
        imageType=NoImage
        mirroring=NoMirror
        resourceName=None
        resourceObject=null
        tiling=NoTile
        imageSize=<0 0>
        margin=<ltrb 0 0 0 0>
        outlineSettings {
         bUseBrushTransparency=false
         roundingType=FixedRadius
         width=1
         color {colorUseRule=UseColor_ColorTable, specifiedColor=<rgba 0 0 0 0>}
         cornerRadii=<4 4 4 4>
        }
        tintColor {colorUseRule=UseColor_Specified, specifiedColor=<rgba 0.0048 0.0048 0.0048 1>}
        uVRegion {bIsValid=false, max=<0 0>, min=<0 0>}
       }
       normalForeground {colorUseRule=UseColor_ColorTable, specifiedColor=<rgba 0 0 0 0>}
       normalPadding=<ltrb 0 0 0 0>
       pressed {
        bIsDynamicallyLoaded=false
        drawAs=RoundedBox
        imageType=NoImage
        mirroring=NoMirror
        resourceName=None
        resourceObject=null
        tiling=NoTile
        imageSize=<0 0>
        margin=<ltrb 0 0 0 0>
        outlineSettings {
         bUseBrushTransparency=false
         roundingType=FixedRadius
         width=1
         color {colorUseRule=UseColor_ColorTable, specifiedColor=<rgba 0 0 0 0>}
         cornerRadii=<4 4 4 4>
        }
        tintColor {colorUseRule=UseColor_Specified, specifiedColor=<rgba 0.0048 0.0048 0.0048 1>}
        uVRegion {bIsValid=false, max=<0 0>, min=<0 0>}
       }
       pressedForeground {colorUseRule=UseColor_ColorTable, specifiedColor=<rgba 0 0 0 0>}
       pressedPadding=<ltrb 0 0 0 0>
       pressedSlateSound {resourceObject=null}
      }
      contentPadding=<ltrb 0 0 0 0>
      downArrowImage {
       bIsDynamicallyLoaded=false
       drawAs=Image
       imageType=Vector
       mirroring=NoMirror
       resourceName=../../../Engine/Content/Slate/Starship/CoreWidgets/ComboBox/wide-chevron-down.svg
       resourceObject=null
       tiling=NoTile
       imageSize=<16 16>
       margin=<ltrb 0 0 0 0>
       outlineSettings {
        bUseBrushTransparency=false
        roundingType=HalfHeightRadius
        width=0
        color {colorUseRule=UseColor_Specified, specifiedColor=<rgba 0 0 0 0>}
        cornerRadii=<0 0 0 0>
       }
       tintColor {colorUseRule=UseColor_Specified, specifiedColor=<rgba 1 1 1 1>}
       uVRegion {bIsValid=false, max=<0 0>, min=<0 0>}
      }
      downArrowPadding=<ltrb 2 2 2 2>
      menuBorderBrush {
       bIsDynamicallyLoaded=false
       drawAs=RoundedBox
       imageType=NoImage
       mirroring=NoMirror
       resourceName=None
       resourceObject=null
       tiling=NoTile
       imageSize=<0 0>
       margin=<ltrb 0 0 0 0>
       outlineSettings {
        bUseBrushTransparency=false
        roundingType=FixedRadius
        width=1
        color {colorUseRule=UseColor_ColorTable, specifiedColor=<rgba 0 0 0 0>}
        cornerRadii=<0 0 0 0>
       }
       tintColor {colorUseRule=UseColor_Specified, specifiedColor=<rgba 0.0395 0.0395 0.0395 1>}
       uVRegion {bIsValid=false, max=<0 0>, min=<0 0>}
      }
      menuBorderPadding=<ltrb 0 0 0 0>
      shadowColorAndOpacity=<rgba 0 0 0 1>
      shadowOffset=<0 0>
     }
     contentPadding=<ltrb 0 0 0 0>
     menuRowPadding=<ltrb 8 3 8 3>
     pressedSlateSound {resourceObject=null}
     selectionChangeSlateSound {resourceObject=null}
    }
   }
   slot {class=SizeBoxSlot}
  }
  SortOrder_ToggleButton {
   widgetClassPath=&EditorUtilityCheckBox
   bIsVariable=true
   parent=HorizontalBox_3
   widget=SortOrder_ToggleButton
   properties {
    checkedState=Checked
    horizontalAlignment=HAlign_Center
    widgetStyle {
     checkBoxType=ToggleButton
     backgroundHoveredImage {
      bIsDynamicallyLoaded=false
      drawAs=RoundedBox
      imageType=NoImage
      mirroring=NoMirror
      resourceName=None
      resourceObject=null
      tiling=NoTile
      imageSize=<18 18>
      margin=<ltrb 0 0 0 0>
      outlineSettings {
       bUseBrushTransparency=false
       roundingType=FixedRadius
       width=1
       color {colorUseRule=UseColor_ColorTable, specifiedColor=<rgba 0 0 0 0>}
       cornerRadii=<3 3 3 3>
      }
      tintColor {colorUseRule=UseColor_Specified, specifiedColor=<rgba 0.0048 0.0048 0.0048 1>}
      uVRegion {bIsValid=false, max=<0 0>, min=<0 0>}
     }
     backgroundImage {
      bIsDynamicallyLoaded=false
      drawAs=RoundedBox
      imageType=NoImage
      mirroring=NoMirror
      resourceName=None
      resourceObject=null
      tiling=NoTile
      imageSize=<18 18>
      margin=<ltrb 0 0 0 0>
      outlineSettings {
       bUseBrushTransparency=false
       roundingType=FixedRadius
       width=1
       color {colorUseRule=UseColor_ColorTable, specifiedColor=<rgba 0 0 0 0>}
       cornerRadii=<3 3 3 3>
      }
      tintColor {colorUseRule=UseColor_Specified, specifiedColor=<rgba 0.0048 0.0048 0.0048 1>}
      uVRegion {bIsValid=false, max=<0 0>, min=<0 0>}
     }
     backgroundPressedImage {
      bIsDynamicallyLoaded=false
      drawAs=RoundedBox
      imageType=NoImage
      mirroring=NoMirror
      resourceName=None
      resourceObject=null
      tiling=NoTile
      imageSize=<18 18>
      margin=<ltrb 0 0 0 0>
      outlineSettings {
       bUseBrushTransparency=false
       roundingType=FixedRadius
       width=1
       color {colorUseRule=UseColor_ColorTable, specifiedColor=<rgba 0 0 0 0>}
       cornerRadii=<3 3 3 3>
      }
      tintColor {colorUseRule=UseColor_Specified, specifiedColor=<rgba 0.0048 0.0048 0.0048 1>}
      uVRegion {bIsValid=false, max=<0 0>, min=<0 0>}
     }
     borderBackgroundColor {colorUseRule=UseColor_Specified, specifiedColor=<rgba 1 1 1 1>}
     checkedForeground {colorUseRule=UseColor_Specified, specifiedColor=<rgba 1 1 1 1>}
     checkedHoveredForeground {colorUseRule=UseColor_Specified, specifiedColor=<rgba 1 1 1 1>}
     checkedHoveredImage {
      bIsDynamicallyLoaded=false
      drawAs=RoundedBox
      imageType=NoImage
      mirroring=NoMirror
      resourceName=None
      resourceObject=null
      tiling=NoTile
      imageSize=<0 0>
      margin=<ltrb 0 0 0 0>
      outlineSettings {
       bUseBrushTransparency=false
       roundingType=FixedRadius
       width=1
       color {colorUseRule=UseColor_Specified, specifiedColor=<rgba 0.0782 0.0782 0.0782 1>}
       cornerRadii=<4 4 4 4>
      }
      tintColor {colorUseRule=UseColor_Specified, specifiedColor=<rgba 0.0048 0.0048 0.0048 1>}
      uVRegion {bIsValid=false, max=<0 0>, min=<0 0>}
     }
     checkedImage {
      bIsDynamicallyLoaded=false
      drawAs=RoundedBox
      imageType=NoImage
      mirroring=NoMirror
      resourceName=None
      resourceObject=null
      tiling=NoTile
      imageSize=<0 0>
      margin=<ltrb 0 0 0 0>
      outlineSettings {
       bUseBrushTransparency=false
       roundingType=FixedRadius
       width=1
       color {colorUseRule=UseColor_Specified, specifiedColor=<rgba 0.0356 0.0356 0.0356 1>}
       cornerRadii=<4 4 4 4>
      }
      tintColor {colorUseRule=UseColor_Specified, specifiedColor=<rgba 0.0048 0.0048 0.0048 1>}
      uVRegion {bIsValid=false, max=<0 0>, min=<0 0>}
     }
     checkedPressedForeground {colorUseRule=UseColor_Specified, specifiedColor=<rgba 1 1 1 1>}
     checkedPressedImage {
      bIsDynamicallyLoaded=false
      drawAs=RoundedBox
      imageType=NoImage
      mirroring=NoMirror
      resourceName=None
      resourceObject=null
      tiling=NoTile
      imageSize=<0 0>
      margin=<ltrb 0 0 0 0>
      outlineSettings {
       bUseBrushTransparency=false
       roundingType=FixedRadius
       width=1
       color {colorUseRule=UseColor_Specified, specifiedColor=<rgba 0.0782 0.0782 0.0782 1>}
       cornerRadii=<4 4 4 4>
      }
      tintColor {colorUseRule=UseColor_Specified, specifiedColor=<rgba 0.0048 0.0048 0.0048 1>}
      uVRegion {bIsValid=false, max=<0 0>, min=<0 0>}
     }
     checkedSlateSound {resourceObject=null}
     foregroundColor {colorUseRule=UseColor_Specified, specifiedColor=<rgba 1 1 1 1>}
     hoveredForeground {colorUseRule=UseColor_Specified, specifiedColor=<rgba 1 1 1 1>}
     hoveredSlateSound {resourceObject=null}
     padding=<ltrb 4 4 4 4>
     pressedForeground {colorUseRule=UseColor_Specified, specifiedColor=<rgba 1 1 1 1>}
     uncheckedHoveredImage {
      bIsDynamicallyLoaded=false
      drawAs=RoundedBox
      imageType=NoImage
      mirroring=NoMirror
      resourceName=None
      resourceObject=null
      tiling=NoTile
      imageSize=<0 0>
      margin=<ltrb 0 0 0 0>
      outlineSettings {
       bUseBrushTransparency=false
       roundingType=FixedRadius
       width=1
       color {colorUseRule=UseColor_Specified, specifiedColor=<rgba 0.0782 0.0782 0.0782 1>}
       cornerRadii=<4 4 4 4>
      }
      tintColor {colorUseRule=UseColor_Specified, specifiedColor=<rgba 0.0048 0.0048 0.0048 1>}
      uVRegion {bIsValid=false, max=<0 0>, min=<0 0>}
     }
     uncheckedImage {
      bIsDynamicallyLoaded=false
      drawAs=RoundedBox
      imageType=NoImage
      mirroring=NoMirror
      resourceName=None
      resourceObject=null
      tiling=NoTile
      imageSize=<0 0>
      margin=<ltrb 0 0 0 0>
      outlineSettings {
       bUseBrushTransparency=false
       roundingType=FixedRadius
       width=1
       color {colorUseRule=UseColor_Specified, specifiedColor=<rgba 0.0356 0.0356 0.0356 1>}
       cornerRadii=<4 4 4 4>
      }
      tintColor {colorUseRule=UseColor_Specified, specifiedColor=<rgba 0.0048 0.0048 0.0048 1>}
      uVRegion {bIsValid=false, max=<0 0>, min=<0 0>}
     }
     uncheckedPressedImage {
      bIsDynamicallyLoaded=false
      drawAs=RoundedBox
      imageType=NoImage
      mirroring=NoMirror
      resourceName=None
      resourceObject=null
      tiling=NoTile
      imageSize=<0 0>
      margin=<ltrb 0 0 0 0>
      outlineSettings {
       bUseBrushTransparency=false
       roundingType=FixedRadius
       width=1
       color {colorUseRule=UseColor_Specified, specifiedColor=<rgba 0.0782 0.0782 0.0782 1>}
       cornerRadii=<4 4 4 4>
      }
      tintColor {colorUseRule=UseColor_Specified, specifiedColor=<rgba 0.0048 0.0048 0.0048 1>}
      uVRegion {bIsValid=false, max=<0 0>, min=<0 0>}
     }
     uncheckedSlateSound {resourceObject=null}
     undeterminedForeground {colorUseRule=UseColor_Specified, specifiedColor=<rgba 1 1 1 1>}
     undeterminedHoveredImage {
      bIsDynamicallyLoaded=false
      drawAs=Image
      imageType=Vector
      mirroring=NoMirror
      resourceName=../../../Engine/Content/Slate/Starship/CoreWidgets/CheckBox/indeterminate.svg
      resourceObject=null
      tiling=NoTile
      imageSize=<16 16>
      margin=<ltrb 0 0 0 0>
      outlineSettings {
       bUseBrushTransparency=false
       roundingType=HalfHeightRadius
       width=0
       color {colorUseRule=UseColor_Specified, specifiedColor=<rgba 0 0 0 0>}
       cornerRadii=<0 0 0 0>
      }
      tintColor {colorUseRule=UseColor_Specified, specifiedColor=<rgba 0.0044 0.2384 1 1>}
      uVRegion {bIsValid=false, max=<0 0>, min=<0 0>}
     }
     undeterminedImage {
      bIsDynamicallyLoaded=false
      drawAs=Image
      imageType=Vector
      mirroring=NoMirror
      resourceName=../../../Engine/Content/Slate/Starship/CoreWidgets/CheckBox/indeterminate.svg
      resourceObject=null
      tiling=NoTile
      imageSize=<16 16>
      margin=<ltrb 0 0 0 0>
      outlineSettings {
       bUseBrushTransparency=false
       roundingType=HalfHeightRadius
       width=0
       color {colorUseRule=UseColor_Specified, specifiedColor=<rgba 0 0 0 0>}
       cornerRadii=<0 0 0 0>
      }
      tintColor {colorUseRule=UseColor_Specified, specifiedColor=<rgba 0 0.162 0.7454 1>}
      uVRegion {bIsValid=false, max=<0 0>, min=<0 0>}
     }
     undeterminedPressedImage {
      bIsDynamicallyLoaded=false
      drawAs=Image
      imageType=Vector
      mirroring=NoMirror
      resourceName=../../../Engine/Content/Slate/Starship/CoreWidgets/CheckBox/indeterminate.svg
      resourceObject=null
      tiling=NoTile
      imageSize=<16 16>
      margin=<ltrb 0 0 0 0>
      outlineSettings {
       bUseBrushTransparency=false
       roundingType=HalfHeightRadius
       width=0
       color {colorUseRule=UseColor_Specified, specifiedColor=<rgba 0 0 0 0>}
       cornerRadii=<0 0 0 0>
      }
      tintColor {colorUseRule=UseColor_Specified, specifiedColor=<rgba 0 0.0802 0.3515 1>}
      uVRegion {bIsValid=false, max=<0 0>, min=<0 0>}
     }
    }
   }
   slot {class=HorizontalBoxSlot, properties {verticalAlignment=VAlign_Center, padding=<ltrb 4 0 4 0>}}
  }
  TechAudioToolsEditorSlateIcon_54 {
   widgetClassPath=/Script/TechAudioToolsContentEditor.TechAudioToolsEditorSlateIcon
   parent=SortOrder_ToggleButton
   widget=TechAudioToolsEditorSlateIcon_54
   properties {fullIconPath=EditorStyle.LandscapeEditor.Target_DisplayOrder.Alphabetical.Ascending, imageColorAndOpacity=<rgba 0.5271 0.5271 0.5271 1>}
   slot {class=PanelSlot}
  }
 ]
}
