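# Editor utility list widget for MetaSound inputs, based on TechAudioToolsEditorListView, with a search and sort bar above the list.
# A dark bordered row holds a search box that fills the width, beside an input list sort menu.
# The inherited ListView fills the remaining space with MetaSound input entry widgets and single-toggle selection.
# The Blueprint only defines layout; its event graph is empty, so no filter or sort wiring appears here.
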EditorUtilityWidgetBlueprint /TechAudioToolsContent/Widgets/Editor/Lists/MetaSoundMemberLists/EUW_MetaSoundInputList_FilterSortOptions
graphs [EventGraph {links [], nodes []}]
variables []
widgets {
 info {
  inheritedWidgetCount=1
  parentClass=/Script/TechAudioToolsContentEditor.TechAudioToolsEditorListView
  rootWidgetClass=&VerticalBox
  widgetCount=6
 }
 widgets [
  VerticalBox_1 {widgetClassPath=&VerticalBox, widget=VerticalBox_1}
  Search_Border {
   widgetClassPath=&Border
   bIsVariable=true
   parent=VerticalBox_1
   widget=Search_Border
   properties {brushColor=<rgba 0.0284 0.0284 0.0284 1>, padding=<ltrb 2 2 2 2>}
   slot {class=VerticalBoxSlot}
  }
  HorizontalBox_1 {
   widgetClassPath=&HorizontalBox
   parent=Search_Border
   widget=HorizontalBox_1
   slot {class=BorderSlot, properties {padding=<ltrb 2 2 2 2>}}
  }
  InputSearchFilter {
   widgetClassPath=/TechAudioToolsContent/Widgets/Editor/EUW_SearchBox
   bIsVariable=true
   parent=HorizontalBox_1
   widget=InputSearchFilter
   slot {class=HorizontalBoxSlot, properties {size {sizeRule=Fill, value=1}}}
  }
  SortOptions {
   widgetClassPath=&/EUW_MetaSoundInputListSortMenu
   bIsVariable=true
   parent=HorizontalBox_1
   widget=SortOptions
   slot {class=HorizontalBoxSlot, properties {horizontalAlignment=HAlign_Left, verticalAlignment=VAlign_Center, padding=<ltrb 2 0 0 0>}}
  }
  ListView {
   widgetClassPath=/Script/Blutility.EditorUtilityListView
   bInherited=true
   bIsVariable=true
   parent=VerticalBox_1
   widget=ListView
   properties {
    bEnableRightClickScrolling=false
    bEnableScrollAnimation=true
    entryWidgetClass=&/EUW_MetaSoundInputEntry.EUW_MetaSoundInputEntry_C
    numDesignerPreviewEntries=15
    selectionMode=SingleToggle
   }
   slot {class=VerticalBoxSlot, properties {size {sizeRule=Fill, value=1}}}
  }
 ]
}
