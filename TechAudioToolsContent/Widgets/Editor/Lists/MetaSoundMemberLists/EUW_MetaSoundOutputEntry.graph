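# Editor utility list entry widget for a MetaSound output in TechAudioTools, derived from TechAudioToolsEditorEntry.
# When a list item object is set, it casts the item to MetaSoundOutputEditorViewModel and assigns it as the widget's viewmodel.
# The layout is a faintly shaded border holding a bullet icon and a fill-width text block with the placeholder "Output Name".
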EditorUtilityWidgetBlueprint /TechAudioToolsContent/Widgets/Editor/Lists/MetaSoundMemberLists/EUW_MetaSoundOutputEntry
graphs [EventGraph {nodes [
   n1 = Event OnListItemObjectSet()
   > n2 = Cast To MetaSoundOutputEditorViewModel!(execute, Object=@n1.ListItemObject) -> (then, CastFailed, AsMetaSound Output Editor Viewmodel:Meta Sound Output Editor Viewmodel Object Reference) {is=title K2Node_DynamicCast}
   > SetMetaSoundOutputEditorViewModel!(execute, self, Viewmodel:Meta Sound Output Editor Viewmodel Object Reference=@n2) -> (then)
  ]}]
variables []
widgets {
 info {parentClass=/Script/TechAudioToolsContentEditor.TechAudioToolsEditorEntry, rootWidgetClass=&VerticalBox, widgetCount=5}
 widgets [
  VerticalBox {widgetClassPath=&VerticalBox, widget=VerticalBox}
  Background {
   widgetClassPath=&Border
   bIsVariable=true
   parent=VerticalBox
   widget=Background
   properties {verticalAlignment=VAlign_Center, brushColor=<rgba 0 0 0 0.1>, padding=<ltrb 0 4 0 4>}
   slot {class=VerticalBoxSlot}
  }
  HorizontalBox_31 {
   widgetClassPath=&HorizontalBox
   parent=Background
   widget=HorizontalBox_31
   slot {class=BorderSlot, properties {verticalAlignment=VAlign_Center, padding=<ltrb 0 4 0 4>}}
  }
  TechAudioToolsEditorSlateIcon_112 {
   widgetClassPath=/Script/TechAudioToolsContentEditor.TechAudioToolsEditorSlateIcon
   parent=HorizontalBox_31
   widget=TechAudioToolsEditorSlateIcon_112
   properties {fullIconPath=EditorStyle.Icons.BulletPoint}
   slot {class=HorizontalBoxSlot, properties {padding=<ltrb 4 0 4 0>}}
  }
  OutputName_Text {
   widgetClassPath=&CommonTextBlock
   parent=HorizontalBox_31
   widget=OutputName_Text
   properties {
    text=Output Name
    colorAndOpacity {colorUseRule=UseColor_Specified, specifiedColor=<rgba 0.5271 0.5271 0.5271 1>}
    font {
     bForceMonospaced=false
     bMaterialIsStencil=false
     fontMaterial=null
     fontObject=&Roboto
     letterSpacing=0
     monospacedWidth=1
     size=9
     skewAmount=0
     typefaceFontName=Regular
     outlineSettings {
      bApplyOutlineToDropShadows=false
      bMiteredCorners=false
      bSeparateFillAlpha=false
      outlineMaterial=null
      outlineSize=0
      outlineColor=<rgba 0 0 0 1>
     }
    }
   }
   slot {class=HorizontalBoxSlot, properties {padding=<ltrb 0 0 6 0>, size {sizeRule=Fill, value=1}}}
  }
 ]
}
